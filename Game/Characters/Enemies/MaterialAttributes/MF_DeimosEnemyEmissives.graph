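# Material function that builds tinted emissive glow for Deimos enemies.
# An RGB mask weights limbs, chest and eyes/mouth with facing-based falloff, plus always-on eye/mouth glow and optional Fresnel overlay.
# Static switches add exploding-brute pulsing with sphere-masked cracks, or ranged-fiend hands using a noise-broken pre-skinned-position gradient blended by vertex green.
# An overall multiplier scales the output; pulsing-multiplier and radius-pulse branches sit unconnected.
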MaterialFunction /Game/Characters/Enemies/MaterialAttributes/MF_DeimosEnemyEmissives
graph {nodes [
  n1 = VectorParameter() {group=Deimos, parameterName=EmisColor, sortPriority=1, defaultValue=<rgba 0.8904 0 0.7673 1>}
  n2 = TextureSampleParameter2D() {
   desc="R = body\r\nG = chest hot glow\r\nB = eyes and mouth"
   group=Deimos
   parameterName=FX Mask
   samplerSource=SSM_Wrap_WorldGroupSettings
   texture=&T_BlacksRGB
  }
  n3 = Clamp(None=DotProduct(A=PixelNormalWS(), B=CameraVectorWS()))
  n4 = Power(Base=@n3, Exp=ScalarParameter() {defaultValue=8, group=Deimos, parameterName=CoreEmisFalloff, sortPriority=6})
  n5 = Multiply(A=@n2.R, B=ScalarParameter() {defaultValue=0.2, group=Deimos, parameterName=Emiss Limbs Strength, sortPriority=3})
  n6 = Multiply(A=@n2.B, B=ScalarParameter() {defaultValue=3, group=Deimos, parameterName=Emiss MouthEye Strength, sortPriority=4})
  n7 = ScalarParameter() {defaultValue=3, group=Deimos, parameterName=Emiss Body Core Strength, sortPriority=5}
  n8 = Multiply(A=Add(A=Add(A=@n5, B=@n6), B=Multiply(A=@n2.G, B=@n7)), B=@n4)
  n9 = ScalarParameter() {defaultValue=1, group=Deimos, parameterName=CoreEmisBright, sortPriority=7}
  n10 = Add(
   A=Power(
    Base=Multiply(A=@n1.RGB, B=@n2.B)
    Exp=ScalarParameter() {defaultValue=4.2, group=Deimos, parameterName=Emiss Always on Power, sortPriority=8}
   )
   B=Multiply(A=@n1.RGB, B=Multiply(A=@n8, B=@n9))
  )
  n11 = Multiply(A=@n1.RGB, B=Power(Base=OneMinus(None=@n3), Exp=ScalarParameter() {defaultValue=2, parameterName=Fresnel_Overlay_Power}))
  n12 = ScalarParameter() {defaultValue=1, parameterName=FresnelOverlay_Strength}
  n13 = LinearInterpolate(A=@n10, B=Add(A=Multiply(A=@n11, B=@n12), B=@n10), Alpha=ScalarParameter() {parameterName=UseFresnelGlowOverlay})
  n14 = PreSkinnedPosition()
  n15 = ShaderStageSwitch(PixelShader=VertexInterpolator(VS=@n14).PS, VertexShader=@n14)
  n16 = ScalarParameter() {defaultValue=64, parameterName=SphereMask_Radius}
  n17 = SphereMask(
   A=@n15
   B=VectorParameter() {parameterName=SphereMaskPosition, defaultValue=<rgba 0 0 100 1>}.RGB
   Radius=@n16
   Hardness=ScalarParameter() {parameterName=SphereMask_Hardness}
  )
  n18 = TextureSampleParameter2D(UVs=Multiply(A=TextureCoordinate(), B=ScalarParameter() {defaultValue=4, parameterName=Cracks_Tiling})) {parameterName=Cracks_Texture, texture=/Game/Effects/Fort_Effects/Textures/T_Cracks_Gradient}
  n19 = Multiply(A=@n1.RGB, B=Multiply(A=@n4, B=Multiply(A=@n17, B=@n18.G)))
  n20 = Time()
  n21 = ScalarParameter() {defaultValue=0.1, parameterName=PulseSpeed}
  n22 = ConstantBiasScale(None=Sine(None=Multiply(A=@n20, B=@n21)))
  n23 = StaticSwitchParameter(True=Multiply(A=Add(A=@n13, B=@n19), B=@n22), False=@n13) {parameterName=UseExplodingBrutePulse}
  n24 = Add(A=@n15, B=VectorParameter() {parameterName=Gradient_Offset, defaultValue=<rgba 0 0 -60 1>}.RGB)
  n25 = OneMinus(None=Divide(
    A=ChannelMaskParameter(None=@n24) {maskChannel=Blue, parameterName=Direction, defaultValue=<rgba 0 0 1 0>}.R
    B=ScalarParameter() {defaultValue=28, parameterName=GradientThreshold}
   ))
  n26 = Multiply(A=TextureCoordinate(), B=ScalarParameter() {defaultValue=12, group=HuskBubble, parameterName=Emissive_NoiseTiling})
  n27 = ComponentMask(None=VectorParameter() {group=HuskBubble, parameterName=DistortionSpeed, defaultValue=<rgba -0.1 0.1 0 1>}.RGB) {g=true, r=true}
  n28 = Multiply(A=TextureSample(UVs=Panner(Coordinate=@n26, Speed=@n27)) {texture=/Game/Global/Textures/Default/Noise/3xPerlin_M}.RGB) {constB=2}
  n29 = Multiply(
   A=ComponentMask(None=Subtract(A=@n28)) {r=true}
   B=ScalarParameter() {defaultValue=0.25, group=HuskBubble, parameterName=EmissiveNoise_Strength}
  )
  n30 = Multiply(A=@n1.RGB, B=Clamp(None=Subtract(A=@n25, B=@n29)))
  n31 = ScalarParameter() {defaultValue=5, parameterName=RangedHandsEmissiveStrength}
  n32 = LinearInterpolate(A=@n23, B=Power(Base=Multiply(A=@n30, B=@n31)), Alpha=VertexColor().G)
  n33 = ScalarParameter() {defaultValue=1, group=Deimos, parameterName=Emis Overall Mult, sortPriority=2}
  FunctionOutput(None=Multiply(A=StaticSwitchParameter(True=@n32, False=@n23) {parameterName=UseRangedFiendEmissiveHands}, B=@n33))
  LinearInterpolate(A=@n33, B=Multiply(A=@n33, B=@n22), Alpha=ScalarParameter() {parameterName=UsePulsingEmissive})
  LinearInterpolate(A=Multiply(A=@n16) {constB=0.25}, B=@n16, Alpha=@n22)
  Multiply(A=@n4, B=@n25)
 ]}
