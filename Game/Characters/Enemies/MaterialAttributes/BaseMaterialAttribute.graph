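# Base material attributes function for Fortnite enemy characters, exposing parameterized texture, tint, specular, roughness and subsurface inputs.
# Base color adds an AO-squared Fresnel rim to tinted diffuse, and the mask's blue channel drives subsurface opacity.
# Switchable emissive uses time-of-day color and an optional sine-pulsed Fresnel, then blends toward Deimos husk emissives.
# Customized UVs 1–4 carry pre-skinned normal, pre-skinned position and the subsurface mask.
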MaterialFunction /Game/Characters/Enemies/MaterialAttributes/BaseMaterialAttribute
graph {
 properties {bExposeToLibrary=true, libraryCategoriesText [Fortnite Parent Materials]}
 nodes [
  n1 = TextureSampleParameter2D() {desc=Diffuse Texture, group=Base Material, parameterName=Diffuse Map, texture=/Engine/EngineMaterials/DefaultDiffuse}
  n2 = Multiply(A=VectorParameter() {group=Base Material, parameterName=Diffuse_Tint, defaultValue=<rgba 1 1 1 1>}.RGB, B=@n1.RGB)
  n3 = TextureSampleParameter2D() {
   desc=Texture Masks
   group=Base Material
   parameterName=Texture Masks
   samplerType=SAMPLERTYPE_Masks
   texture=/Engine/EngineMaterials/DefaultDiffuse_TC_Masks
  }
  n4 = Constant() {r=3}
  n5 = VertexNormalWS()
  n6 = Multiply(
   A=Multiply(A=@n3.R, B=@n3.R) {desc=ambient occlusion}
   B=Multiply(
    A=ScalarParameter() {defaultValue=0.6, desc=Fresnel Brightness, group=Base Material, parameterName=Fresnel Brightness}
    B=Fresnel(ExponentIn=@n4, Normal=@n5)
   )
  )
  n7 = TextureSampleParameter2D() {
   desc=R: Spec G: Metal B: Roughness
   group=Base Material
   parameterName=Specular Masks
   samplerType=SAMPLERTYPE_Masks
   texture=&T_Fortnite_Default_S
  }
  n8 = ScalarParameter() {defaultValue=0.5, group=Base Material, parameterName=Specular Scale}
  n9 = ScalarParameter() {group=Base Material, parameterName=RoughnessMin}
  n10 = ScalarParameter() {defaultValue=1, group=Base Material, parameterName=RoughnessMax}
  n11 = Multiply(A=ScalarParameter() {defaultValue=1, group=Base Material, parameterName=SSS Opacity}, B=@n3.B)
  n12 = TextureSampleParameter2D() {
   desc=Normal Map
   group=Base Material
   parameterName=Normal Map
   samplerType=SAMPLERTYPE_Normal
   texture=/Engine/EngineMaterials/DefaultNormal
  }
  n13 = Multiply(A=@n1.RGB, B=VectorParameter() {group=Base Material, parameterName=SSS Color, defaultValue=<rgba 1 1 1 1>}.RGB)
  n14 = PreSkinnedNormal()
  n15 = ComponentMask(None=@n14) {b=true}
  n16 = PreSkinnedPosition()
  n17 = ComponentMask(None=@n16) {r=true}
  n18 = Constant()
  n19 = Multiply(A=ScalarParameter() {defaultValue=0.4145, group=Base Emissive, parameterName=Multiply Time}, B=Time() {period=0.01})
  n20 = ScalarParameter() {defaultValue=0.1823, group=Base Emissive, parameterName=Fresnel Eponent}
  n21 = ScalarParameter() {defaultValue=0.0474, group=Base Emissive, parameterName=Fresnel BaseReflection}
  n22 = LinearInterpolate(Alpha=Multiply(A=Sine(None=@n19), B=Fresnel(ExponentIn=@n20, BaseReflectFractionIn=@n21))) {constA=1, constB=0}
  n23 = TextureSampleParameter2D() {desc=SphereMap Texture, group=Base Material, parameterName=Emissive Map, texture=&T_BlacksRGB}
  n24 = Multiply(
   A=@n23.R
   B=TextureSampleParameter2D() {desc=Diffuse Texture, group=Base Material, parameterName=Diffuse Map, texture=&DefaultDiffuse}.RGB
  )
  n25 = VectorParameter() {group=Base Emissive, parameterName=Color_Morning, defaultValue=<rgba 0.2625 0.3501 0.5 0>}
  n26 = VectorParameter() {group=Base Emissive, parameterName=Color_Evening, defaultValue=<rgba 0.3934 0.1729 1 0>}
  n27 = VectorParameter() {group=Base Emissive, parameterName=Color_Night, defaultValue=<rgba 1 0.0276 0.771 0>}
  n28 = Multiply(
   A=StaticSwitchParameter(True=@n24, False=@n24) {group=Base Material, parameterName=EmissiveMapMultipliedAgainstDiffuse}
   B=Multiply(
    A=MaterialFunctionCall(Color_Morning=@n25.RGB, Color_Evening=@n26.RGB, Color_Night=@n27.RGB) {materialFunction=&Color_Time_of_Day}.Result
    B=VectorParameter() {group=Base Emissive, parameterName=Emissive Scale, defaultValue=<rgba 1 1 1 0>}.RGB
   )
  )
  n29 = StaticSwitchParameter(
   True=StaticSwitchParameter(True=Multiply(A=@n22, B=@n28), False=@n28) {group=Base Material, parameterName=Use Emissive Fresnel}
   False=Constant()
  ) {group=Base Material, parameterName=Use Emissive Texture}
  n30 = MaterialFunctionCall() {materialFunction=&/MF_DeimosEnemyEmissives}
  n31 = ScalarParameter() {group=Deimos, parameterName=Is Deimos Husk}
  FunctionOutput(None=MakeMaterialAttributes!(
    BaseColor=Add(A=@n2, B=@n6)
    Metallic=@n7.G
    Specular=Multiply(A=@n8, B=@n7.R)
    Roughness=LinearInterpolate(A=@n9, B=@n10, Alpha=@n7.B)
    Anisotropy
    EmissiveColor=LinearInterpolate(A=@n29, B=@n30.Result, Alpha=@n31)
    Opacity=@n11
    OpacityMask=ScalarParameter() {defaultValue=1, group=Base Material, parameterName=Opacity Mask}
    Normal=@n12.RGB
    Tangent
    WorldPositionOffset
    SubsurfaceColor=Multiply(A=@n13, B=@n11)
    ClearCoat
    ClearCoatRoughness
    AmbientOcclusion=@n3.R
    Refraction
    CustomizedUVs_0
    CustomizedUVs_1=ComponentMask(None=@n14) {g=true, r=true}
    CustomizedUVs_2=AppendVector(A=@n15, B=@n17)
    CustomizedUVs_3=ComponentMask(None=@n16) {b=true, g=true}
    CustomizedUVs_4=AppendVector(A=@n3.B, B=@n18)
    CustomizedUVs_5
    CustomizedUVs_6
    CustomizedUVs_7
    PixelDepthOffset
    ShadingModel
    Displacement
   ))
 ]
}
