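# MaterialAttributeTester is a library-exposed material function that outputs a test checkerboard as a full Material Attributes set, stored with the enemy character material attributes.
# A tiled CheckerPattern on default UVs drives a lerp between dark blue and pale blue constants.
# The blend feeds only BaseColor in MakeMaterialAttributes, leaving every other attribute at its default.
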MaterialFunction /Game/Characters/Enemies/MaterialAttributes/MaterialAttributeTester
graph {
 properties {bExposeToLibrary=true, libraryCategoriesText [Fortnite Parent Materials]}
 nodes [
  n1 = Constant3Vector() {constant=<rgba 0.0281 0.0769 0.26 0>}
  n2 = Constant3Vector() {constant=<rgba 0.4061 0.5543 0.7 0>}
  n3 = MaterialFunctionCall!(UVs=TextureCoordinate(), Tiling=Constant() {r=20}) {materialFunction=/Engine/ArtTools/RenderToTexture/MaterialFunctions/CheckerPattern}
  FunctionOutput(None=MakeMaterialAttributes!(
    BaseColor=LinearInterpolate(A=@n1, B=@n2, Alpha=@n3.Result)
    Metallic
    Specular
    Roughness
    Anisotropy
    EmissiveColor
    Opacity
    OpacityMask
    Normal
    Tangent
    WorldPositionOffset
    SubsurfaceColor
    ClearCoat
    ClearCoatRoughness
    AmbientOcclusion
    Refraction
    CustomizedUVs_0
    CustomizedUVs_1
    CustomizedUVs_2
    CustomizedUVs_3
    CustomizedUVs_4
    CustomizedUVs_5
    CustomizedUVs_6
    CustomizedUVs_7
    PixelDepthOffset
    ShadingModel
    Displacement
   ))
 ]
}
