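# Material function that adds a death dissolve to enemy materials, taking and returning Material Attributes.
# A SphereMask at HitLocation, over a position read from UV channels 2 and 3, grows with DeathDissolve and combines with tiled noise to erode OpacityMask.
# The erosion band drives hot and soft emissive glows, with desert cracks detailing the leading edge.
# A faster saturated DeathDissolve ramp darkens BaseColor, adds crack normals and blends other surface attributes.
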MaterialFunction /Game/Characters/Enemies/MaterialAttributes/DissolveEnemiesMaterialAttribute
graph {
 properties {bExposeToLibrary=true, libraryCategoriesText [Fortnite Parent Materials]}
 nodes [
  ConstantBiasScale(None=Sine(None=Time()) {period=5})
  n1 = GetMaterialAttributes(None=FunctionInput(Preview=MaterialFunctionCall() {materialFunction=&/MaterialAttributeTester}.Result) {inputName=MA In, inputType=FunctionInput_MaterialAttributes, previewValue=<0 0 0 0>}) {attributeGetTypes [
    #1
    #2
    #3
    #4
    #5
    #6
    #7
    #8
    #9
   ]}
  n2 = VectorParameter() {group=Dissolve, parameterName=Dark Color, defaultValue=<rgba 0.015 0.0068 0.0022 0>}
  n3 = ScalarParameter() {group=Dissolve, parameterName=DeathDissolve}
  n4 = Saturate(None=Multiply(A=@n3) {constB=4.5})
  n5 = TextureSample(UVs=TextureCoordinate() {uTiling=4, vTiling=4}) {
   samplerSource=SSM_Wrap_WorldGroupSettings
   samplerType=SAMPLERTYPE_Normal
   texture=/Game/Global/Textures/Default/Tileable_Patterns/T_Desert_Cracks_N
  }
  n6 = Add(A=@n5.R, B=@n5.R)
  n7 = Multiply(A=ScalarParameter() {defaultValue=0.5, group=Dissolve, parameterName=Leading Cracks Opacity}, B=Power(Base=Saturate(None=@n6)))
  n8 = ScalarParameter() {defaultValue=0.6, group=Dissolve, parameterName=Dissolve U HQ}
  n9 = ScalarParameter() {defaultValue=0.6, group=Dissolve, parameterName=Dissolve V HQ}
  n10 = TextureSample(UVs=Multiply(A=TextureCoordinate(), B=AppendVector(A=@n8, B=@n9))) {samplerType=SAMPLERTYPE_Normal, texture=/Game/Global/Textures/Default/Tileable_Patterns/2ndInfectedMapNormal}
  n11 = ComponentMask(None=TextureCoordinate() {coordinateIndex=2}) {g=true}
  n12 = TextureCoordinate() {coordinateIndex=3}
  n13 = ScalarParameter() {defaultValue=150, group=Dissolve, parameterName=Object Radius}
  n14 = OneMinus(None=SphereMask(
    A=AppendVector(A=@n11, B=@n12)
    B=VectorParameter() {group=Dissolve, parameterName=HitLocation, defaultValue=<rgba 0 0 80 0>}.RGB
    Radius=Multiply(A=@n3, B=@n13)
    Hardness=ScalarParameter() {defaultValue=0.5, group=Dissolve, parameterName=DissolveHardness}
   ))
  n15 = LinearInterpolate(A=Multiply(A=ConstantBiasScale(None=@n10.R), B=@n14), Alpha=@n14)
  n16 = OneMinus(None=@n15)
  n17 = Saturate(None=Multiply(A=Multiply(A=@n16, B=@n15)) {constB=4})
  n18 = Power(Base=@n17, Exp=ScalarParameter() {defaultValue=50, group=Dissolve, parameterName=Emissive Soft Cracks Exponent})
  n19 = ScalarParameter() {defaultValue=900, group=Dissolve, parameterName=Emissive Hot Exponent}
  n20 = Multiply(
   A=Add(A=Multiply(A=@n7, B=@n18), B=Power(Base=@n17, Exp=@n19) {constExponent=200})
   B=VectorParameter() {group=Dissolve, parameterName=Emissive Hot, defaultValue=<rgba 5.443 2.3015 40 0>}.RGB
  )
  n21 = ScalarParameter() {defaultValue=2, group=Dissolve, parameterName=Emissive Soft Exponent}
  n22 = ScalarParameter() {defaultValue=0.55, group=Dissolve, parameterName=EmissiveDiffuseDesaturation}
  n23 = Multiply(
   A=Multiply(A=Power(Base=@n17, Exp=@n21), B=Desaturation(None=@n1.BaseColor, Fraction=@n22))
   B=VectorParameter() {group=Dissolve, parameterName=Emissive Soft Color, defaultValue=<rgba 0.1437 0.084 0.8 0>}.RGB
  )
  n24 = ScalarParameter() {defaultValue=1.5, group=Dissolve, parameterName=Dissolve Contrast}
  n25 = Add(A=@n1.Normal, B=@n5.RGB)
  n26 = SetMaterialAttributes!(
   MaterialAttributes=@n1.MaterialAttributes
   BaseColor=LinearInterpolate(A=@n1.BaseColor, B=Multiply(A=@n1.BaseColor, B=@n2.RGB), Alpha=@n4)
   Metallic=LinearInterpolate(A=@n1.Metallic, B=ScalarParameter() {group=Dissolve, parameterName=Metallic Dissolve}, Alpha=@n4) {constB=0}
   Specular=LinearInterpolate(A=@n1.Specular, B=ScalarParameter() {defaultValue=0.5, group=Dissolve, parameterName=Specular Dissolve}, Alpha=@n4) {constB=0.5}
   Roughness=LinearInterpolate(A=@n1.Roughness, B=ScalarParameter() {defaultValue=0.6, group=Dissolve, parameterName=Roughness Dissolve}, Alpha=@n4) {constB=0.6}
   EmissiveColor=LinearInterpolate(A=@n1.EmissiveColor, B=Add(A=@n20, B=@n23), Alpha=@n4) {constB=0}
   Opacity=LinearInterpolate(A=@n1.Opacity, Alpha=@n4) {constB=0}
   OpacityMask=Multiply(A=Power(Base=@n15, Exp=@n24), B=@n1.OpacityMask)
   Normal=Normalize(VectorInput=LinearInterpolate(A=@n1.Normal, B=@n25, Alpha=@n4))
   SubsurfaceColor=LinearInterpolate(A=@n1.SubsurfaceColor, Alpha=@n4) {constB=0}
  ) {attributeSetTypes [
    #1
    #2
    #3
    #4
    #5
    #6
    #7
    #8
    #9
   ]}
  FunctionOutput(None=@n26) {outputName=MA Out}
 ]
}
