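# Material attribute function for enemy characters that adds an eye glow on top of incoming material attributes.
# An outer and inner eye color are blended by a Fresnel effect driven by the transformed incoming normal, with brightness and exponent parameters.
# The result is masked by vertex color green times the EyeGlowMask scalar, then added to existing emissive, leaving other attributes untouched.
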MaterialFunction /Game/Characters/Enemies/MaterialAttributes/EnemyEyeGlowMaterialAttribute
graph {
 properties {bExposeToLibrary=true, libraryCategoriesText [Fortnite Parent Materials]}
 nodes [
  n1 = GetMaterialAttributes(None=FunctionInput(Preview=MaterialFunctionCall() {materialFunction=&/MaterialAttributeTester}.Result) {inputName=MA In, inputType=FunctionInput_MaterialAttributes}) {attributeGetTypes [#1, #2]}
  n2 = ScalarParameter() {defaultValue=1, group=Base Material, parameterName=Eye Fresnel Brightness}
  n3 = ScalarParameter() {defaultValue=5, group=Base Material, parameterName=Eye Fresnel Exponent}
  n4 = Transform(None=@n1.Normal)
  n5 = Multiply(
   A=Multiply(A=VertexColor().G, B=ScalarParameter() {defaultValue=1, parameterName=EyeGlowMask})
   B=LinearInterpolate(
    A=VectorParameter() {group=Base Material, parameterName=Eye Color Outer, defaultValue=<rgba 2.25 0.5422 0.0225 1>}.RGB
    B=VectorParameter() {group=Base Material, parameterName=Eye Color Inner, defaultValue=<rgba 5 4.0325 1.8529 1>}.RGB
    Alpha=MaterialFunctionCall(Brightness=@n2, Exponent=@n3, Vertex / Pixel Normal=@n4) {materialFunction=&MF_Fresnel_FX}.Fresnel
   ) {constA=1, constB=0}
  )
  FunctionOutput(None=SetMaterialAttributes!(MaterialAttributes=@n1.MaterialAttributes, EmissiveColor=Add(A=@n1.EmissiveColor, B=@n5)) {attributeSetTypes [#1]}) {outputName=MA Out}
 ]
}
