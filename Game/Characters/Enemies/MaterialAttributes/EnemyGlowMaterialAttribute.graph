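# Material function that layers an enemy hit glow and wiggle onto incoming material attributes, exposed under Fortnite Parent Materials.
# A Fresnel term blends HitGlowOuter and HitGlowInner colors, and the red channel of the HitGlow parameter fades that into Emissive.
# HitGlow alpha scales a time-driven sine, remapped by ConstantBiasScale, that pushes vertices along world normals by HitWiggleAmount, added to existing World Position Offset.
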MaterialFunction /Game/Characters/Enemies/MaterialAttributes/EnemyGlowMaterialAttribute
graph {
 properties {bExposeToLibrary=true, libraryCategoriesText [Fortnite Parent Materials]}
 nodes [
  n1 = GetMaterialAttributes(None=FunctionInput(Preview=MaterialFunctionCall() {materialFunction=&/MaterialAttributeTester}.Result) {inputName=MA In, inputType=FunctionInput_MaterialAttributes}) {attributeGetTypes [#1, #2]}
  n2 = ScalarParameter() {defaultValue=1.4, group=Enemy Glow, parameterName=HitGlowFresnelBrightness}
  n3 = ScalarParameter() {defaultValue=3, group=Enemy Glow, parameterName=HitGlowFresnelExponent}
  n4 = Add(
   A=@n1.EmissiveColor
   B=LinearInterpolate(
    A=VectorParameter() {group=Enemy Glow, parameterName=HitGlowOuter, defaultValue=<rgba 0.175 0.0767 0.0196 1>}.RGB
    B=VectorParameter() {group=Enemy Glow, parameterName=HitGlowInner, defaultValue=<rgba 0.05 0.0129 0.0055 1>}.RGB
    Alpha=MaterialFunctionCall(Brightness=@n2, Exponent=@n3) {materialFunction=&MF_Fresnel_FX}.Fresnel
   )
  )
  n5 = VectorParameter() {group=Enemy Glow, parameterName=HitGlow}
  n6 = Time()
  n7 = ScalarParameter() {defaultValue=7.5, group=Enemy Hit Wiggle, parameterName=HitWiggleSpeed}
  n8 = ConstantBiasScale(None=Sine(None=Multiply(A=@n6, B=@n7)))
  n9 = VertexNormalWS()
  n10 = Multiply(A=@n5.A, B=Multiply(A=Multiply(A=@n8, B=@n9), B=ScalarParameter() {desc=1.75, group=Enemy Hit Wiggle, parameterName=HitWiggleAmount}))
  FunctionOutput(None=SetMaterialAttributes!(
    MaterialAttributes=@n1.MaterialAttributes
    EmissiveColor=LinearInterpolate(A=@n1.EmissiveColor, B=@n4, Alpha=@n5.R)
    WorldPositionOffset=Add(A=@n1.WorldPositionOffset, B=@n10)
   ) {attributeSetTypes [#1, #2]})
 ]
}
