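# Fiend enemy material function that layers time-driven spawn and death effects onto incoming material attributes.
# Spawn twists vertices by bounding-box height and grows them out from the bounds' bottom as a lightning-colored glow fades; Low quality drops the twist.
# Behind the UseDeathFX static switch, death stretches vertices away from that base, weighted by height, while noise-driven HeightLerp dissolves the opacity mask with the same glow.
# Start-time and length parameters drive both phases.
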MaterialFunction /Game/Characters/Enemies/MaterialAttributes/MF_FiendMaterialFX
graph {nodes [
  n1 = GetMaterialAttributes(None=FunctionInput(Preview=SetMaterialAttributes!(MaterialAttributes)) {bUsePreviewValueAsDefault=true, inputName=Material, inputType=FunctionInput_MaterialAttributes}) {attributeGetTypes [#1, #2]}
  n2 = Constant3Vector() {constant=<rgba 0 0 -1 0>}
  n3 = ObjectRadius()
  n4 = Constant3Vector() {constant=<rgba 1 0 0 0>}
  n5 = Subtract(A=Time(), B=ScalarParameter() {defaultValue=-100, group=Deimos, parameterName=Spawn_CurrentTime, sortPriority=21})
  n6 = ScalarParameter() {defaultValue=0.6, group=Deimos, parameterName=Spawn_AnimLength, sortPriority=22}
  n7 = OneMinus(None=Clamp(None=Divide(A=@n5, B=@n6)))
  n8 = MaterialFunctionCall() {materialFunction=&BoundingBoxBased_0-1_UVW}
  n9 = RotateAboutAxis(
   NormalizedRotationAxis=Transform(None=Normalize(VectorInput=@n4)) {transformSourceType=TRANSFORMSOURCE_Local}
   RotationAngle=Multiply(A=Multiply(A=@n7, B=@n8.B)) {constB=-2}
   PivotPoint=ActorPositionWS()
   Position=WorldPosition().XYZ
  )
  n10 = Subtract(A=Add(A=ActorPositionWS(), B=Multiply(A=@n2, B=@n3)), B=Add(A=WorldPosition().XYZ, B=@n9))
  n11 = Power(Base=@n7, Exp=Constant() {r=0.5})
  n12 = Add(A=ActorPositionWS(), B=Multiply(A=Constant3Vector() {constant=<rgba 0 0 -1 0>}, B=ObjectRadius()))
  n13 = WorldPosition()
  n14 = QualitySwitch(Default=Add(A=Multiply(A=@n10, B=@n11), B=@n9), Low=Multiply(A=Subtract(A=@n12, B=@n13.XYZ), B=@n7))
  n15 = VectorParameter() {group=Deimos, parameterName=SpawnLighteningColor, sortPriority=20, defaultValue=<rgba 12 0.2281 8.4659 1>}
  n16 = Multiply(A=@n7, B=@n15.RGB)
  n17 = GetMaterialAttributes(None=SetMaterialAttributes!(
    MaterialAttributes=@n1.MaterialAttributes
    WorldPositionOffset=Add(A=@n1.WorldPositionOffset, B=@n14)
    EmissiveColor=Add(A=@n1.EmissiveColor, B=@n16)
    OpacityMask
   ) {attributeSetTypes [#2, #1, #3]}) {attributeGetTypes [#1, #2]}
  n18 = Divide(
   A=Subtract(A=Time(), B=ScalarParameter() {defaultValue=-100, group=Deimos, parameterName=Death_CurrentTime, sortPriority=23})
   B=ScalarParameter() {defaultValue=0.7, group=Deimos, parameterName=Death_AnimLength, sortPriority=24}
  )
  n19 = OneMinus(None=@n18)
  n20 = Multiply(A=Clamp(None=@n18), B=Clamp(None=Add(A=@n19) {constB=10}))
  n21 = ActorPositionWS()
  n22 = Multiply(A=Constant3Vector() {constant=<rgba 0 0 -1 0>}, B=ObjectRadius())
  n23 = Multiply(A=Subtract(A=WorldPosition().XYZ, B=Add(A=@n21, B=@n22))) {constB=0.5}
  n24 = Add(A=@n17.WorldPositionOffset, B=Multiply(A=Power(Base=@n20) {constExponent=4}, B=Multiply(A=@n23, B=@n8.B)))
  n25 = MaterialFunctionCall(
   Transition Phase=LinearInterpolate(Alpha=@n20) {constA=1, constB=0}
   Height Texture=TextureSample() {samplerSource=SSM_Wrap_WorldGroupSettings, texture=/Game/Global/Textures/Default/Noise/T_Dudebro_Noise_01}.R
  ) {materialFunction=&HeightLerp}
  n26 = Add(A=@n17.EmissiveColor, B=Multiply(A=@n15.RGB, B=OneMinus(None=@n25.Alpha)))
  FunctionOutput(None=StaticSwitchParameter(
    True=SetMaterialAttributes(MaterialAttributes=@n17.MaterialAttributes, WorldPositionOffset=@n24, EmissiveColor=@n26, OpacityMask=@n25.Alpha) {attributeSetTypes [#2, #1, #3]}
    False=@n17.MaterialAttributes
   ) {defaultValue=true, parameterName=UseDeathFX})
  OneMinus(None=@n8.B)
  Power(Base=@n7)
 ]}
