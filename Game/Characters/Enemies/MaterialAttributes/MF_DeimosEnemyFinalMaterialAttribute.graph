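# MF_DeimosEnemyFinalMaterialAttribute is a material function that serves as a final attribute pass for Deimos enemy materials.
# It takes a Material Attributes input and runs it through the EnemyGlowMaterialAttribute function.
# That result then feeds MF_FiendMaterialFX, whose output becomes the function's single output, chaining enemy glow and fiend effects into one reusable node.
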MaterialFunction /Game/Characters/Enemies/MaterialAttributes/MF_DeimosEnemyFinalMaterialAttribute
graph {nodes [
  n1 = FunctionInput() {inputName=Material, inputType=FunctionInput_MaterialAttributes}
  FunctionOutput(None=MaterialFunctionCall!(Material=MaterialFunctionCall!(MA In=@n1) {materialFunction=&/EnemyGlowMaterialAttribute}.Result) {materialFunction=&/MF_FiendMaterialFX}.Result)
 ]}
