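# Material attributes function for enemy spawn-in effects, layered over incoming attributes to animate color and vertex position.
# SpawnColorAnimation darkens base color and blends emissive toward a two-color glow patterned from contrast-adjusted base color.
# SpawnAnimation drives world position offset toward SpawnLocationWS, staggered by HeightLerp using each vertex's projection along that direction over object radius.
# A DynamicBranch on SpawnAnimation selects between zero and that offset.
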MaterialFunction /Game/Characters/Enemies/MaterialAttributes/SpawnInEnemiesMaterialAttribute
graph {
 properties {bExposeToLibrary=true, libraryCategoriesText [Fortnite Parent Materials]}
 nodes [
  n1 = GetMaterialAttributes(None=FunctionInput(Preview=MaterialFunctionCall() {materialFunction=&/MaterialAttributeTester}.Result) {inputName=MA In, inputType=FunctionInput_MaterialAttributes}) {attributeGetTypes [#1, #2, #3]}
  n2 = ScalarParameter() {group=Spawn In, parameterName=SpawnColorAnimation}
  n3 = OneMinus(None=@n2)
  n4 = Power(
   Base=LinearInterpolate(
    A=VectorParameter() {group=Spawn In, parameterName=Spawn-In Diffuse Low, defaultValue=<rgba 1 1 1 0>}.RGB
    B=VectorParameter() {group=Spawn In, parameterName=Spawn-In Diffuse High, defaultValue=<rgba 0.3 0.3 0.3 0>}.RGB
    Alpha=@n1.BaseColor
   )
   Exp=ScalarParameter() {defaultValue=13, group=Spawn In, parameterName=Spawn-In Contrast}
  )
  n5 = VectorParameter() {group=Spawn In, parameterName=Spawn-in Glow Color Low, defaultValue=<rgba 0.4248 0.267 5 0>}
  n6 = VectorParameter() {group=Spawn In, parameterName=Spawn-In Glow Color High, defaultValue=<rgba 1.5493 1.12 14 0>}
  n7 = Saturate(None=Subtract(A=@n4, B=@n3))
  n8 = Multiply(A=@n4, B=Multiply(A=@n2, B=LinearInterpolate(A=@n5.RGB, B=@n6.RGB, Alpha=@n7)))
  n9 = ScalarParameter() {group=Spawn In, parameterName=SpawnAnimation}
  n10 = Subtract(
   A=VectorParameter() {group=Spawn In, parameterName=SpawnLocationWS, defaultValue=<rgba 0 0 0 1>}.RGB
   B=WorldPosition() {worldPositionShaderOffset=WPT_ExcludeAllShaderOffsets}.XYZ
  )
  n11 = Normalize(VectorInput=@n10)
  n12 = Subtract(
   A=WorldPosition() {worldPositionShaderOffset=WPT_ExcludeAllShaderOffsets}.XYZ
   B=MaterialFunctionCall() {materialFunction=&ComponentPivotLocation}.Result
  )
  n13 = Clamp(None=Divide(A=DotProduct(A=@n11, B=@n12), B=ObjectRadius()))
  n14 = MaterialFunctionCall!(
   Alpha=@n9
   Color Input 0=Constant()
   Color Input 1=Multiply(
    A=MaterialFunctionCall(Transition Phase=@n9, Height Texture=@n13) {materialFunction=&HeightLerp}.Lerp Alpha No Contrast
    B=@n10
   )
  ) {materialFunction=/Engine/Functions/Engine_MaterialFunctions02/Utility/DynamicBranch}
  FunctionOutput(None=SetMaterialAttributes!(
    MaterialAttributes=@n1.MaterialAttributes
    BaseColor=Multiply(A=@n1.BaseColor, B=@n3)
    EmissiveColor=LinearInterpolate(A=@n8, B=@n1.EmissiveColor, Alpha=@n3)
    WorldPositionOffset=LinearInterpolate(A=@n1.WorldPositionOffset, B=@n14.Result, Alpha=@n9)
   ) {attributeSetTypes [#1, #2, #3]}) {outputName=MA Out}
 ]
}
