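# Material function that modifies incoming enemy material attributes into a fire elemental look with animated emissive flames.
# It triplanar-projects fire mask and normal textures using a position built from UV channels 2 and 3, panned over time, with local-space vertex normals.
# Mask channels feed tinted base color, specular and emissive.
# Inverted AO, mask, fresnel and base-color brightness blend two fire colors into added emissive, while the projected normal perturbs the original.
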MaterialFunction /Game/Characters/Enemies/MaterialModifiers/Elemental_Fire/MA_Elemental_Fire
graph {nodes [
  n1 = GetMaterialAttributes(None=FunctionInput(Preview=MaterialFunctionCall() {materialFunction=/Game/Characters/Enemies/MaterialAttributes/MaterialAttributeTester}.Result) {description=spec color goes into world position, inputName=MA, inputType=FunctionInput_MaterialAttributes}) {attributeGetTypes [
    #1
    #2
    #3
    #4
    #5
   ]}
  n2 = VectorParameter() {desc="160, 160, 100", group=ElementalFire, parameterName=FireTextureSizeWS, defaultValue=<rgba 200 200 125 0>}
  n3 = TextureCoordinate() {coordinateIndex=2}
  n4 = Time() {bOverride_Period=true, period=50}
  n5 = Time() {bOverride_Period=true, period=10.4167}
  n6 = Add(
   A=AppendVector(A=ComponentMask(None=@n3) {g=true}, B=TextureCoordinate() {coordinateIndex=3})
   B=Multiply(
    A=MaterialFunctionCall(R=@n4, G=@n4, B=@n5) {materialFunction=&AppendMany}.RGB
    B=VectorParameter() {group=ElementalFire, parameterName=FirePannerSpeedWS, defaultValue=<rgba 0 -4 12 0>}.RGB
   )
  )
  n7 = Transform(None=VertexNormalWS()) {transformSourceType=TRANSFORMSOURCE_World, transformType=TRANSFORM_Local}
  n8 = ScalarParameter() {defaultValue=2, group=ElementalFire, parameterName=FireWSProjectionContrast}
  n9 = MaterialFunctionCall(
   TextureObject=TextureObjectParameter() {group=ElementalFire, parameterName=FireBaseColor, texture=/Game/Characters/Enemies/Textures/T_Elemental_Fire_M}
   TextureSize=@n2.RGB
   WorldPosition=@n6
   World Space Normal=@n7
   ProjectionTransitionContrast=@n8
  ) {materialFunction=&WorldAlignedTexture}
  n10 = ComponentMask(None=@n9.XYZ Texture) {b=true}
  n11 = ComponentMask(None=@n1.Opacity) {b=true}
  n12 = ComponentMask(None=@n9.XYZ Texture) {g=true}
  n13 = Power(Base=OneMinus(None=@n1.AmbientOcclusion), Exp=ScalarParameter() {defaultValue=1.4, group=ElementalFire, parameterName=FireAOExponent})
  n14 = ComponentMask(None=@n9.XYZ Texture) {r=true}
  n15 = ScalarParameter() {defaultValue=0.75, group=ElementalFire, parameterName=FireFresnelBrightness}
  n16 = ScalarParameter() {defaultValue=11, group=ElementalFire, parameterName=FireFresnelExponent}
  n17 = VertexNormalWS()
  n18 = Desaturation(None=@n1.BaseColor)
  n19 = ScalarParameter() {defaultValue=6, group=ElementalFire, parameterName=FireBaseColorExponent}
  n20 = Add(
   A=Add(
    A=Add(A=@n13, B=@n14)
    B=MaterialFunctionCall(Brightness=@n15, Exponent=@n16, Vertex / Pixel Normal=@n17) {materialFunction=&MF_Fresnel_FX}.Fresnel
   )
   B=Multiply(A=Power(Base=@n18, Exp=@n19) {constExponent=3}, B=ScalarParameter() {defaultValue=60, group=ElementalFire, parameterName=FireBaseColorScale})
  ) {constB=0}
  n21 = LinearInterpolate(
   A=VectorParameter() {group=ElementalFire, parameterName=FireColorA, defaultValue=<rgba 1.25 0.0217 0 0>}.RGB
   B=VectorParameter() {group=ElementalFire, parameterName=FireColorB, defaultValue=<rgba 2 0.4896 0 0>}.RGB
   Alpha=@n20
  )
  n22 = Saturate(None=@n20)
  n23 = Constant()
  n24 = TextureObjectParameter() {group=ElementalFire, parameterName=FireNormal, samplerType=SAMPLERTYPE_Normal, texture=/Game/Maps/Environments/T_Wasteland_RockWaveey_N}
  n25 = Multiply(
   A=MaterialFunctionCall(TextureObject=@n24, TextureSize=@n2.RGB, WorldPosition=@n6, World Space Normal=@n7, ProjectionTransitionContrast=@n8) {materialFunction=&WorldAlignedTexture}.XYZ Texture
   B=VectorParameter() {group=ElementalFire, parameterName=FireNormalIntensity, defaultValue=<rgba 1 1 0 1>}.RGB
  )
  n26 = SetMaterialAttributes!(
   MaterialAttributes=@n1.MaterialAttributes
   BaseColor=Multiply(
    A=LinearInterpolate(A=@n1.BaseColor, B=@n10, Alpha=@n11)
    B=VectorParameter() {group=ElementalFire, parameterName=FireBaseColor, defaultValue=<rgba 0.13 0.0433 0.0433 0>}.RGB
   )
   Metallic=ScalarParameter() {defaultValue=0.05, group=ElementalFire, parameterName=FireMetallic}
   Specular=LinearInterpolate(
    A=ScalarParameter() {defaultValue=0.05, group=ElementalFire, parameterName=FireSpecLow}
    B=ScalarParameter() {defaultValue=0.7, group=ElementalFire, parameterName=FireSpecHigh}
    Alpha=LinearInterpolate(B=@n12, Alpha=@n11)
   )
   Roughness=ScalarParameter() {defaultValue=0.3, group=ElementalFire, parameterName=FireRoughness}
   EmissiveColor=Add(A=@n1.EmissiveColor, B=LinearInterpolate(B=@n21, Alpha=@n22) {constB=0})
   Opacity=@n23
   Normal=Normalize(VectorInput=Add(A=@n1.Normal, B=@n25))
   SubsurfaceColor=@n23
  ) {attributeSetTypes [
    #1
    #6
    #7
    #8
    #2
    #3
    #4
    #9
   ]}
  FunctionOutput(None=@n26)
 ]}
