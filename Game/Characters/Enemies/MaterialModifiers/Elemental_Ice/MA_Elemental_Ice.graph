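# Material function that turns an enemy's material attributes into an Elemental Ice variant with frostbitten hands and lower body.
# It desaturates and blue-tints base color, using CustomizedUV4 red to separate clothing from icy areas broken up by smoke noise.
# A frostbite mask combines height and hand sphere masks on pre-skinned positions read from UV2/UV3, blending color, metallic, specular and roughness.
# A fixed view-space Fresnel adds masked emissive glow.
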MaterialFunction /Game/Characters/Enemies/MaterialModifiers/Elemental_Ice/MA_Elemental_Ice
graph {nodes [
  n1 = FunctionInput(Preview=MaterialFunctionCall() {materialFunction=/Game/Characters/Enemies/MaterialAttributes/MaterialAttributeTester}.Result) {bUsePreviewValueAsDefault=true, description=spec color goes into world position, inputName=MA, inputType=FunctionInput_MaterialAttributes}
  n2 = GetMaterialAttributes(None=@n1) {attributeGetTypes [
    #1
    #2
    #3
    #4
    #5
    #6
    #7
    #8
    #9
   ]}
  n3 = Add(
   A=ScalarParameter() {defaultValue=0.06, group=ElementalIce, parameterName=IceBaseColorBias}
   B=Desaturation(None=@n2.BaseColor, Fraction=ScalarParameter() {defaultValue=0.85, group=ElementalIce, parameterName=IceBaseColorDesaturation})
  ) {constA=0.1}
  n4 = Multiply(A=VectorParameter() {group=ElementalIce, parameterName=IceClothingTint, defaultValue=<rgba 0.0256 0.0886 0.195 0>}.RGB, B=@n3)
  n5 = ComponentMask(None=@n2.CustomizedUV4) {r=true}
  n6 = LinearInterpolate(
   A=@n4
   B=Multiply(A=@n3, B=VectorParameter() {group=ElementalIce, parameterName=IceBaseColorTint, defaultValue=<rgba 1.3275 2.6208 4 0>}.RGB)
   Alpha=@n5
  )
  n7 = TextureCoordinate()
  n8 = ScalarParameter() {group=ElementalIce, parameterName=IceNoiseUVRandom}
  n9 = TextureSample(UVs=Multiply(A=Add(A=@n7, B=@n8), B=ScalarParameter() {defaultValue=6, group=ElementalIce, parameterName=IceNoiseUVTile})) {texture=/Game/Global/Textures/Default/Tileable_Patterns/T_Smoke_OUT}
  n10 = Round(None=@n9.R)
  n11 = ScalarParameter() {defaultValue=0.4, group=ElementalIce, parameterName=IceNoiseLERP}
  n12 = Multiply(
   A=Add(A=LinearInterpolate(A=@n9.RGB, B=@n10, Alpha=@n11), B=ScalarParameter() {defaultValue=0.8, group=ElementalIce, parameterName=IceNoiseBias})
   B=ScalarParameter() {defaultValue=1.2, group=ElementalIce, parameterName=IceNoiseScale}
  )
  n13 = Add(
   A=ComponentMask(None=TextureCoordinate() {coordinateIndex=3}) {g=true}
   B=ScalarParameter() {defaultValue=-5, group=ElementalIce, parameterName=FrostbiteHeightBias}
  )
  n14 = ScalarParameter() {defaultValue=25, group=ElementalIce, parameterName=FrostbiteHeightScale}
  n15 = VectorParameter() {group=ElementalIce, parameterName=FrostbiteHandL-PreSkinLS, defaultValue=<rgba 40.0611 -5 77.9089 0>}
  n16 = AppendVector(A=ComponentMask(None=TextureCoordinate() {coordinateIndex=2}) {g=true}, B=TextureCoordinate() {coordinateIndex=3})
  n17 = ScalarParameter() {defaultValue=50, group=ElementalIce, parameterName=FrostbiteHandRadius}
  n18 = ScalarParameter() {defaultValue=0.15, group=ElementalIce, parameterName=FrostbiteHandHardness}
  n19 = VectorParameter() {group=ElementalIce, parameterName=FrostbiteHandR-PreSkinLS, defaultValue=<rgba -40.0611 -5 77.9089 0>}
  n20 = Multiply(
   A=Add(A=SphereMask(A=@n15.RGB, B=@n16, Radius=@n17, Hardness=@n18), B=SphereMask(A=@n19.RGB, B=@n16, Radius=@n17, Hardness=@n18))
   B=ScalarParameter() {defaultValue=0.97, group=ElementalIce, parameterName=FrostbiteHandsScale}
  )
  n21 = Saturate(None=Multiply(A=Divide(A=@n13, B=@n14), B=OneMinus(None=@n20)))
  n22 = ScalarParameter() {defaultValue=4, group=ElementalIce, parameterName=IceFresnelExponent}
  n23 = Transform(None=VectorParameter() {group=ElementalIce, parameterName="IceViewSpaceCameraVector ", defaultValue=<rgba -0.5 -0.2 -1 0>}.RGB) {transformSourceType=TRANSFORMSOURCE_View}
  n24 = Transform(None=@n2.Normal)
  n25 = Multiply(
   A=Multiply(
    A=MaterialFunctionCall(Exponent=@n22, Camera Vector=@n23, Vertex / Pixel Normal=@n24) {materialFunction=&MF_Fresnel_FX}.Fresnel Inverted
    B=VectorParameter() {group=ElementalIce, parameterName=IceViewSpaceFresnelColor, defaultValue=<rgba 0.096 0.2677 0.6 0>}.RGB
   )
   B=@n21
  )
  n26 = SetMaterialAttributes!(
   MaterialAttributes=@n2.MaterialAttributes
   BaseColor=LinearInterpolate(A=@n4, B=Multiply(A=@n6, B=@n12), Alpha=@n21)
   Metallic=LinearInterpolate(A=ScalarParameter() {defaultValue=0.03, group=ElementalIce, parameterName=IceFrostbiteMetallic}, B=@n2.Metallic, Alpha=@n21)
   Specular=LinearInterpolate(A=ScalarParameter() {defaultValue=0.5, group=ElementalIce, parameterName=IceFrostbiteSpecular}, B=@n2.Specular, Alpha=@n21)
   Roughness=LinearInterpolate(A=ScalarParameter() {defaultValue=0.5, group=ElementalIce, parameterName=IceFrostbiteRoughness}, B=@n2.Roughness, Alpha=@n21)
   EmissiveColor=Add(A=@n2.EmissiveColor, B=Multiply(A=@n5, B=@n25))
   Opacity=LinearInterpolate(B=@n2.Opacity, Alpha=@n5)
   SubsurfaceColor=LinearInterpolate(B=@n2.SubsurfaceColor, Alpha=@n5)
  ) {attributeSetTypes [
    #1
    #2
    #3
    #4
    #5
    #6
    #8
   ]}
  FunctionOutput(None=@n26)
 ]}
