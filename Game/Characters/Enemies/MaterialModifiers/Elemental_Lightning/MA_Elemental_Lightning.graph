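# Material function that gives enemy materials an electric elemental glow by modifying incoming material attributes.
# Two summed sines of differing periods make the inner glow color flicker, and a mask blends it toward a steady outer color.
# That saturated mask sums powered desaturated base color, inverted AO and inverted Fresnel.
# The blend adds to existing emissive, while opacity and subsurface color inputs receive a zero constant.
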MaterialFunction /Game/Characters/Enemies/MaterialModifiers/Elemental_Lightning/MA_Elemental_Lightning
graph {nodes [
  n1 = FunctionInput(Preview=MaterialFunctionCall() {materialFunction=/Game/Characters/Enemies/MaterialAttributes/MaterialAttributeTester}.Result) {bUsePreviewValueAsDefault=true, description=spec color goes into world position, inputName=MA, inputType=FunctionInput_MaterialAttributes}
  n2 = GetMaterialAttributes(None=@n1) {attributeGetTypes [
    #1
    #2
    #3
    #4
    #5
    #6
   ]}
  n3 = Time()
  n4 = Add(A=Sine(None=@n3) {period=0.2229}, B=Sine(None=@n3) {period=0.2973})
  n5 = ScalarParameter() {defaultValue=0.2, group=ElementalElectric, parameterName=ElectricSineScale}
  n6 = Multiply(
   A=VectorParameter() {group=ElementalElectric, parameterName=ElectricGlowColorInner, defaultValue=<rgba 0.095 0.0158 0 0>}.RGB
   B=Add(A=Multiply(A=@n4, B=@n5)) {constB=0.5}
  )
  n7 = VectorParameter() {group=ElementalElectric, parameterName=ElectricGlowColorOuter, defaultValue=<rgba 2 0.9153 0.0278 0>}
  n8 = Desaturation(None=@n2.BaseColor, Fraction=ScalarParameter() {defaultValue=0.8, group=ElementalElectric, parameterName=ElectricBaseColorDesaturation})
  n9 = ScalarParameter() {defaultValue=2.5, group=ElementalElectric, parameterName=ElectricDiffuseToEmissiveExponent}
  n10 = Multiply(A=OneMinus(None=@n2.AmbientOcclusion), B=ScalarParameter() {defaultValue=0.4, group=ElementalElectric, parameterName=ElectricAOBrightness})
  n11 = MaterialFunctionCall(
   Brightness=ScalarParameter() {defaultValue=0.8, group=ElementalElectric, parameterName=ElectricFresnelBrightness}
   Exponent=ScalarParameter() {defaultValue=4, group=ElementalElectric, parameterName=ElectricFresnelExponent}
  ) {materialFunction=&MF_Fresnel_FX}
  n12 = Saturate(None=Add(A=Power(Base=@n8, Exp=@n9), B=Add(A=@n10, B=@n11.Fresnel Inverted)))
  n13 = Constant()
  n14 = SetMaterialAttributes!(
   MaterialAttributes=@n2.MaterialAttributes
   BaseColor
   Metallic=@n2.MaterialAttributes
   Specular=@n2.MaterialAttributes
   Roughness=@n2.MaterialAttributes
   EmissiveColor=Add(A=@n2.EmissiveColor, B=LinearInterpolate(A=@n6, B=@n7.RGB, Alpha=@n12))
   Opacity=@n13
   SubsurfaceColor=@n13
  ) {attributeSetTypes [
    #1
    #2
    #3
    #4
    #5
    #7
    #8
   ]}
  FunctionOutput(None=@n14)
 ]}
